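# Anim Blueprint for the UpbeatIguanaElm female head that adds secondary physics motion to its hair, eye and zipper bones.
# The input pose feeds two cached AnimDynamics chains over the hair and eye bones: root-relative for in-game, world-space with retuned eye bodies for the front end.
# bIsInFrontEnd selects between them, then a world-space zipper simulation runs on the result.
# Hair dampens and clamps component motion; the event graph holds no logic.
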AnimBlueprint /BRCosmetics/Characters/Player/Female/Medium/Heads/F_MED_UpbeatIguanaElm_Head/Meshes/CustomSkeleton/F_MED_Head_UpbeatIguanaElm_AnimB
target_skeleton=&/Fortnite_Base_Head_UpbeatIguanaElm_Sk
graphs [
 AnimGraph {nodes [
   n1 = Use cached pose 'inGameCache'()
   n2 = Use cached pose 'inFrontEndCache'()
   n3 = Get bIsInFrontEnd()
   n4 = Anim Dynamics - Bone: dyn_zipper01!(
    ComponentPose=Local To Component(LocalPose=Blend Poses by bool(BlendPose_0=@n2, BlendPose_1=@n1, bActiveValue=@n3) {blendType=Linear})
    Alpha=1
    GravityOverride:Vector="0,0,-980"
   ) -> (Pose) {
    angularDampingOverride=1
    angularSpringConstant=20000
    bAngularSpring=true
    bOverrideAngularDamping=true
    bUseGravityOverride=true
    gravityScale=0.35
    simulationSpace=World
    alphaBoolBlend {blendOption=Linear}
    boundBone {boneName=dyn_zipper01}
    componentLinearAccScale=<1 1 1>
    componentLinearVelScale=<1 1 1>
    gravityOverride {z=-980}
    physicsBodyDefinitions [{
      boundBone {boneName=dyn_zipper01}
      boxExtents=<10 30 30>
      constraintSetup {angularLimitsMax=<15 0 15>, angularLimitsMin=<-15 -70 -15>, angularTarget=<1 0 0>}
      localJointOffset=<4 0 0>
     }]
   }
   Output Pose(Result=Component To Local(ComponentPose=@n4))
   Save cached pose 'InputPoseCache'(Pose=Input Pose()) {cacheName=InputPoseCache}
   n5 = Anim Dynamics - Bone: dyn_hairA_01!(ComponentPose=Local To Component(LocalPose=Use cached pose 'InputPoseCache'()), Alpha=1, GravityOverride:Vector="0,0,-980") -> (Pose) {
    angularDampingOverride=1
    angularSpringConstant=10000
    bAngularSpring=true
    bOverrideAngularDamping=true
    bUseGravityOverride=true
    gravityScale=0.35
    simulationSpace=RootRelative
    alphaBoolBlend {blendOption=Linear}
    boundBone {boneName=dyn_hairA_01}
    componentAppliedLinearAccClamp=<3000 3000 3000>
    componentLinearAccScale=<0.1 0.1 0.1>
    componentLinearVelScale=<0.1 0.1 0.1>
    gravityOverride {z=-980}
    physicsBodyDefinitions [{
      boundBone {boneName=dyn_hairA_01}
      boxExtents=<14.5 24.5 24.5>
      constraintSetup {angularLimitsMax=<15 15 15>, angularLimitsMin=<-15 -15 -15>, angularTarget=<1 0 0>}
      localJointOffset=<0.8 0 0>
     }]
   }
   n6 = Anim Dynamics - Bone: dyn_hairA_02!(ComponentPose=@n5, Alpha=1, GravityOverride:Vector="0,0,-980") -> (Pose) {
    angularDampingOverride=1
    angularSpringConstant=10000
    bAngularSpring=true
    bOverrideAngularDamping=true
    bUseGravityOverride=true
    gravityScale=0.35
    simulationSpace=RootRelative
    alphaBoolBlend {blendOption=Linear}
    boundBone {boneName=dyn_hairA_02}
    componentAppliedLinearAccClamp=<3000 3000 3000>
    componentLinearAccScale=<0.1 0.1 0.1>
    componentLinearVelScale=<0.1 0.1 0.1>
    gravityOverride {z=-980}
    physicsBodyDefinitions [{
      boundBone {boneName=dyn_hairA_02}
      boxExtents=<14.5 24.5 24.5>
      constraintSetup {angularLimitsMax=<15 15 8>, angularLimitsMin=<-15 -15 -8>, angularTarget=<1 0 0>}
      localJointOffset=<0.8 0 0>
     }]
   }
   n7 = Anim Dynamics - Bone: dyn_hairB_01!(ComponentPose=@n6, Alpha=1, GravityOverride:Vector="0,0,-980") -> (Pose) {
    angularDampingOverride=1
    angularSpringConstant=10000
    bAngularSpring=true
    bOverrideAngularDamping=true
    bUseGravityOverride=true
    gravityScale=0.35
    simulationSpace=RootRelative
    alphaBoolBlend {blendOption=Linear}
    boundBone {boneName=dyn_hairB_01}
    componentAppliedLinearAccClamp=<3000 3000 3000>
    componentLinearAccScale=<0.1 0.1 0.1>
    componentLinearVelScale=<0.1 0.1 0.1>
    gravityOverride {z=-980}
    physicsBodyDefinitions [{
      boundBone {boneName=dyn_hairB_01}
      boxExtents=<15 25 25>
      constraintSetup {angularLimitsMax=<15 15 15>, angularLimitsMin=<-15 -15 -15>, angularTarget=<1 0 0>}
      localJointOffset=<0.8 0 0>
     }]
   }
   n8 = Anim Dynamics - Bone: dyn_hairB_02!(ComponentPose=@n7, Alpha=1, GravityOverride:Vector="0,0,-980") -> (Pose) {
    angularDampingOverride=1
    angularSpringConstant=10000
    bAngularSpring=true
    bOverrideAngularDamping=true
    bUseGravityOverride=true
    gravityScale=0.35
    simulationSpace=RootRelative
    alphaBoolBlend {blendOption=Linear}
    boundBone {boneName=dyn_hairB_02}
    componentAppliedLinearAccClamp=<3000 3000 3000>
    componentLinearAccScale=<0.1 0.1 0.1>
    componentLinearVelScale=<0.1 0.1 0.1>
    gravityOverride {z=-980}
    physicsBodyDefinitions [{
      boundBone {boneName=dyn_hairB_02}
      boxExtents=<15 25 25>
      constraintSetup {angularLimitsMax=<15 15 8>, angularLimitsMin=<-15 -15 -8>, angularTarget=<1 0 0>}
      localJointOffset=<0.8 0 0>
     }]
   }
   n9 = Anim Dynamics - Bone: dyn_hairC_01!(ComponentPose=@n8, Alpha=1, GravityOverride:Vector="0,0,-980") -> (Pose) {
    angularDampingOverride=1
    angularSpringConstant=10000
    bAngularSpring=true
    bOverrideAngularDamping=true
    bUseGravityOverride=true
    gravityScale=0.35
    simulationSpace=RootRelative
    alphaBoolBlend {blendOption=Linear}
    boundBone {boneName=dyn_hairC_01}
    componentAppliedLinearAccClamp=<3000 3000 3000>
    componentLinearAccScale=<0.1 0.1 0.1>
    componentLinearVelScale=<0.1 0.1 0.1>
    gravityOverride {z=-980}
    physicsBodyDefinitions [{
      boundBone {boneName=dyn_hairC_01}
      boxExtents=<15.5 25.5 25.5>
      constraintSetup {angularLimitsMax=<15 15 15>, angularLimitsMin=<-15 -15 -15>, angularTarget=<1 0 0>}
      localJointOffset=<0.8 0 0>
     }]
   }
   n10 = Anim Dynamics - Bone: dyn_hairC_02!(ComponentPose=@n9, Alpha=1, GravityOverride:Vector="0,0,-980") -> (Pose) {
    angularDampingOverride=1
    angularSpringConstant=10000
    bAngularSpring=true
    bOverrideAngularDamping=true
    bUseGravityOverride=true
    gravityScale=0.35
    simulationSpace=RootRelative
    alphaBoolBlend {blendOption=Linear}
    boundBone {boneName=dyn_hairC_02}
    componentAppliedLinearAccClamp=<3000 3000 3000>
    componentLinearAccScale=<0.1 0.1 0.1>
    componentLinearVelScale=<0.1 0.1 0.1>
    gravityOverride {z=-980}
    physicsBodyDefinitions [{
      boundBone {boneName=dyn_hairC_02}
      boxExtents=<15.5 25.5 25.5>
      constraintSetup {angularLimitsMax=<15 15 8>, angularLimitsMin=<-15 -15 -8>, angularTarget=<1 0 0>}
      localJointOffset=<0.8 0 0>
     }]
   }
   n11 = Anim Dynamics - Bone: dyn_eye_01_L!(ComponentPose=@n10, Alpha=1) -> (Pose) {
    angularDampingOverride=1
    angularSpringConstant=20000
    bAngularSpring=true
    bOverrideAngularDamping=true
    bUseGravityOverride=true
    gravityScale=0.35
    simulationSpace=RootRelative
    alphaBoolBlend {blendOption=Linear}
    boundBone {boneName=dyn_eye_01_L}
    componentLinearAccScale=<1 1 1>
    componentLinearVelScale=<1 1 1>
    gravityOverride {z=-980}
    physicsBodyDefinitions [{
      boundBone {boneName=dyn_eye_01_L}
      boxExtents=<4 4 40>
      constraintSetup {angularLimitsMax=<15 25 25>, angularLimitsMin=<-15 -25 -25>, angularTarget=<1 0 0>}
      localJointOffset=<2 0 0>
     }]
   }
   n12 = Anim Dynamics - Bone: dyn_eye_01_R!(ComponentPose=@n11, Alpha=1) -> (Pose) {
    angularDampingOverride=1
    angularSpringConstant=20000
    bAngularSpring=true
    bOverrideAngularDamping=true
    bUseGravityOverride=true
    gravityScale=0.35
    simulationSpace=RootRelative
    alphaBoolBlend {blendOption=Linear}
    boundBone {boneName=dyn_eye_01_R}
    componentLinearAccScale=<1 1 1>
    componentLinearVelScale=<1 1 1>
    gravityOverride {z=-980}
    physicsBodyDefinitions [{
      boundBone {boneName=dyn_eye_01_R}
      boxExtents=<20 20 20>
      constraintSetup {angularLimitsMax=<15 35 35>, angularLimitsMin=<-15 -35 -35>, angularTarget=<1 0 0>}
      localJointOffset=<2 0 0>
     }]
   }
   n13 = Anim Dynamics - Bone: dyn_eye_02_L!(ComponentPose=@n12, Alpha=1) -> (Pose) {
    angularDampingOverride=1
    angularSpringConstant=20000
    bAngularSpring=true
    bOverrideAngularDamping=true
    bUseGravityOverride=true
    gravityScale=0.35
    simulationSpace=RootRelative
    alphaBoolBlend {blendOption=Linear}
    boundBone {boneName=dyn_eye_02_L}
    componentLinearAccScale=<1 1 1>
    componentLinearVelScale=<1 1 1>
    gravityOverride {z=-980}
    physicsBodyDefinitions [{
      boundBone {boneName=dyn_eye_02_L}
      boxExtents=<20 30 30>
      constraintSetup {angularLimitsMax=<15 30 25>, angularLimitsMin=<-15 -30 -30>, angularTarget=<1 0 0>}
      localJointOffset=<2 0 0>
     }]
   }
   n14 = Anim Dynamics - Bone: dyn_eye_02_R!(ComponentPose=@n13, Alpha=1) -> (Pose) {
    angularDampingOverride=1
    angularSpringConstant=20000
    bAngularSpring=true
    bOverrideAngularDamping=true
    bUseGravityOverride=true
    gravityScale=0.35
    simulationSpace=RootRelative
    alphaBoolBlend {blendOption=Linear}
    boundBone {boneName=dyn_eye_02_R}
    componentLinearAccScale=<1 1 1>
    componentLinearVelScale=<1 1 1>
    gravityOverride {z=-980}
    physicsBodyDefinitions [{
      boundBone {boneName=dyn_eye_02_R}
      boxExtents=<20 30 30>
      constraintSetup {angularLimitsMax=<15 30 30>, angularLimitsMin=<-15 -30 -30>, angularTarget=<1 0 0>}
      localJointOffset=<2 0 0>
     }]
   }
   n15 = Anim Dynamics - Bone: dyn_eye_03_L!(ComponentPose=@n14, Alpha=1) -> (Pose) {
    angularDampingOverride=1
    angularSpringConstant=20000
    bAngularSpring=true
    bOverrideAngularDamping=true
    bUseGravityOverride=true
    gravityScale=0.35
    simulationSpace=RootRelative
    alphaBoolBlend {blendOption=Linear}
    boundBone {boneName=dyn_eye_03_L}
    componentLinearAccScale=<1 1 1>
    componentLinearVelScale=<1 1 1>
    gravityOverride {z=-980}
    physicsBodyDefinitions [{
      boundBone {boneName=dyn_eye_03_L}
      boxExtents=<20 20 20>
      constraintSetup {angularLimitsMax=<5 40 40>, angularLimitsMin=<-15 -40 -40>, angularTarget=<1 0 0>}
      localJointOffset=<2 0 0>
     }]
   }
   n16 = Anim Dynamics - Bone: dyn_eye_03_R!(ComponentPose=@n15, Alpha=1) -> (Pose) {
    angularDampingOverride=1
    angularSpringConstant=20000
    bAngularSpring=true
    bOverrideAngularDamping=true
    bUseGravityOverride=true
    gravityScale=0.35
    simulationSpace=RootRelative
    alphaBoolBlend {blendOption=Linear}
    boundBone {boneName=dyn_eye_03_R}
    componentLinearAccScale=<1 1 1>
    componentLinearVelScale=<1 1 1>
    gravityOverride {z=-980}
    physicsBodyDefinitions [{
      boundBone {boneName=dyn_eye_03_R}
      boxExtents=<20 30 30>
      constraintSetup {angularLimitsMax=<15 15 15>, angularLimitsMin=<-15 -15 -15>, angularTarget=<1 0 0>}
      localJointOffset=<2 0 0>
     }]
   }
   Save cached pose 'inGameCache'(Pose=Component To Local(ComponentPose=@n16)) {cacheName=inGameCache}
   n17 = Anim Dynamics - Bone: dyn_hairA_01(ComponentPose=Local To Component(LocalPose=Use cached pose 'InputPoseCache'()), Alpha=1, GravityOverride="0,0,-980") {
    angularDampingOverride=1
    angularSpringConstant=10000
    bAngularSpring=true
    bOverrideAngularDamping=true
    bUseGravityOverride=true
    gravityScale=0.35
    simulationSpace=World
    alphaBoolBlend {blendOption=Linear}
    boundBone {boneName=dyn_hairA_01}
    componentAppliedLinearAccClamp=<3000 3000 3000>
    componentLinearAccScale=<0.1 0.1 0.1>
    componentLinearVelScale=<0.1 0.1 0.1>
    gravityOverride {z=-980}
    physicsBodyDefinitions [{
      boundBone {boneName=dyn_hairA_01}
      boxExtents=<14.5 24.5 24.5>
      constraintSetup {angularLimitsMax=<15 15 15>, angularLimitsMin=<-15 -15 -15>, angularTarget=<1 0 0>}
      localJointOffset=<0.8 0 0>
     }]
   }
   n18 = Anim Dynamics - Bone: dyn_hairA_02(ComponentPose=@n17, Alpha=1, GravityOverride="0,0,-980") {
    angularDampingOverride=1
    angularSpringConstant=10000
    bAngularSpring=true
    bOverrideAngularDamping=true
    bUseGravityOverride=true
    gravityScale=0.35
    simulationSpace=World
    alphaBoolBlend {blendOption=Linear}
    boundBone {boneName=dyn_hairA_02}
    componentAppliedLinearAccClamp=<3000 3000 3000>
    componentLinearAccScale=<0.1 0.1 0.1>
    componentLinearVelScale=<0.1 0.1 0.1>
    gravityOverride {z=-980}
    physicsBodyDefinitions [{
      boundBone {boneName=dyn_hairA_02}
      boxExtents=<14.5 24.5 24.5>
      constraintSetup {angularLimitsMax=<15 15 8>, angularLimitsMin=<-15 -15 -8>, angularTarget=<1 0 0>}
      localJointOffset=<0.8 0 0>
     }]
   }
   n19 = Anim Dynamics - Bone: dyn_hairB_01(ComponentPose=@n18, Alpha=1, GravityOverride="0,0,-980") {
    angularDampingOverride=1
    angularSpringConstant=10000
    bAngularSpring=true
    bOverrideAngularDamping=true
    bUseGravityOverride=true
    gravityScale=0.35
    simulationSpace=World
    alphaBoolBlend {blendOption=Linear}
    boundBone {boneName=dyn_hairB_01}
    componentAppliedLinearAccClamp=<3000 3000 3000>
    componentLinearAccScale=<0.1 0.1 0.1>
    componentLinearVelScale=<0.1 0.1 0.1>
    gravityOverride {z=-980}
    physicsBodyDefinitions [{
      boundBone {boneName=dyn_hairB_01}
      boxExtents=<15 25 25>
      constraintSetup {angularLimitsMax=<15 15 15>, angularLimitsMin=<-15 -15 -15>, angularTarget=<1 0 0>}
      localJointOffset=<0.8 0 0>
     }]
   }
   n20 = Anim Dynamics - Bone: dyn_hairB_02(ComponentPose=@n19, Alpha=1, GravityOverride="0,0,-980") {
    angularDampingOverride=1
    angularSpringConstant=10000
    bAngularSpring=true
    bOverrideAngularDamping=true
    bUseGravityOverride=true
    gravityScale=0.35
    simulationSpace=World
    alphaBoolBlend {blendOption=Linear}
    boundBone {boneName=dyn_hairB_02}
    componentAppliedLinearAccClamp=<3000 3000 3000>
    componentLinearAccScale=<0.1 0.1 0.1>
    componentLinearVelScale=<0.1 0.1 0.1>
    gravityOverride {z=-980}
    physicsBodyDefinitions [{
      boundBone {boneName=dyn_hairB_02}
      boxExtents=<15 25 25>
      constraintSetup {angularLimitsMax=<15 15 8>, angularLimitsMin=<-15 -15 -8>, angularTarget=<1 0 0>}
      localJointOffset=<0.8 0 0>
     }]
   }
   n21 = Anim Dynamics - Bone: dyn_hairC_01(ComponentPose=@n20, Alpha=1, GravityOverride="0,0,-980") {
    angularDampingOverride=1
    angularSpringConstant=10000
    bAngularSpring=true
    bOverrideAngularDamping=true
    bUseGravityOverride=true
    gravityScale=0.35
    simulationSpace=World
    alphaBoolBlend {blendOption=Linear}
    boundBone {boneName=dyn_hairC_01}
    componentAppliedLinearAccClamp=<3000 3000 3000>
    componentLinearAccScale=<0.1 0.1 0.1>
    componentLinearVelScale=<0.1 0.1 0.1>
    gravityOverride {z=-980}
    physicsBodyDefinitions [{
      boundBone {boneName=dyn_hairC_01}
      boxExtents=<15.5 25.5 25.5>
      constraintSetup {angularLimitsMax=<15 15 15>, angularLimitsMin=<-15 -15 -15>, angularTarget=<1 0 0>}
      localJointOffset=<0.8 0 0>
     }]
   }
   n22 = Anim Dynamics - Bone: dyn_hairC_02(ComponentPose=@n21, Alpha=1, GravityOverride="0,0,-980") {
    angularDampingOverride=1
    angularSpringConstant=10000
    bAngularSpring=true
    bOverrideAngularDamping=true
    bUseGravityOverride=true
    gravityScale=0.35
    simulationSpace=World
    alphaBoolBlend {blendOption=Linear}
    boundBone {boneName=dyn_hairC_02}
    componentAppliedLinearAccClamp=<3000 3000 3000>
    componentLinearAccScale=<0.1 0.1 0.1>
    componentLinearVelScale=<0.1 0.1 0.1>
    gravityOverride {z=-980}
    physicsBodyDefinitions [{
      boundBone {boneName=dyn_hairC_02}
      boxExtents=<15.5 25.5 25.5>
      constraintSetup {angularLimitsMax=<15 15 8>, angularLimitsMin=<-15 -15 -8>, angularTarget=<1 0 0>}
      localJointOffset=<0.8 0 0>
     }]
   }
   n23 = Anim Dynamics - Bone: dyn_eye_01_L(ComponentPose=@n22, Alpha=1) {
    angularDampingOverride=1
    angularSpringConstant=15000
    bAngularSpring=true
    bOverrideAngularDamping=true
    bUseGravityOverride=true
    gravityScale=0.35
    simulationSpace=World
    alphaBoolBlend {blendOption=Linear}
    boundBone {boneName=dyn_eye_01_L}
    componentLinearAccScale=<1 1 1>
    componentLinearVelScale=<1 1 1>
    gravityOverride {z=-980}
    physicsBodyDefinitions [{
      boundBone {boneName=dyn_eye_01_L}
      boxExtents=<20 25 25>
      constraintSetup {angularLimitsMax=<10 25 25>, angularLimitsMin=<-10 -25 -25>, angularTarget=<1 0 0>}
      localJointOffset=<1 0 0>
     }]
   }
   n24 = Anim Dynamics - Bone: dyn_eye_01_R(ComponentPose=@n23, Alpha=1) {
    angularDampingOverride=1
    angularSpringConstant=15000
    bAngularSpring=true
    bOverrideAngularDamping=true
    bUseGravityOverride=true
    gravityScale=0.35
    simulationSpace=World
    alphaBoolBlend {blendOption=Linear}
    boundBone {boneName=dyn_eye_01_R}
    componentLinearAccScale=<1 1 1>
    componentLinearVelScale=<1 1 1>
    gravityOverride {z=-980}
    physicsBodyDefinitions [{
      boundBone {boneName=dyn_eye_01_R}
      boxExtents=<20 25 25>
      constraintSetup {angularLimitsMax=<10 25 25>, angularLimitsMin=<-10 -25 -25>, angularTarget=<1 0 0>}
      localJointOffset=<1 0 0>
     }]
   }
   n25 = Anim Dynamics - Bone: dyn_eye_02_L(ComponentPose=@n24, Alpha=1) {
    angularDampingOverride=1
    angularSpringConstant=18000
    bAngularSpring=true
    bOverrideAngularDamping=true
    bUseGravityOverride=true
    gravityScale=0.35
    simulationSpace=World
    alphaBoolBlend {blendOption=Linear}
    boundBone {boneName=dyn_eye_02_L}
    componentLinearAccScale=<1 1 1>
    componentLinearVelScale=<1 1 1>
    gravityOverride {z=-980}
    physicsBodyDefinitions [{
      boundBone {boneName=dyn_eye_02_L}
      boxExtents=<22 22 22>
      constraintSetup {angularLimitsMax=<10 25 25>, angularLimitsMin=<-10 -25 -25>, angularTarget=<1 0 0>}
      localJointOffset=<1 0 0>
     }]
   }
   n26 = Anim Dynamics - Bone: dyn_eye_02_R(ComponentPose=@n25, Alpha=1) {
    angularDampingOverride=1
    angularSpringConstant=18000
    bAngularSpring=true
    bOverrideAngularDamping=true
    bUseGravityOverride=true
    gravityScale=0.35
    simulationSpace=World
    alphaBoolBlend {blendOption=Linear}
    boundBone {boneName=dyn_eye_02_R}
    componentLinearAccScale=<1 1 1>
    componentLinearVelScale=<1 1 1>
    gravityOverride {z=-980}
    physicsBodyDefinitions [{
      boundBone {boneName=dyn_eye_02_R}
      boxExtents=<22 22 22>
      constraintSetup {angularLimitsMax=<10 25 25>, angularLimitsMin=<-10 -25 -25>, angularTarget=<1 0 0>}
      localJointOffset=<1 0 0>
     }]
   }
   n27 = Anim Dynamics - Bone: dyn_eye_03_L(ComponentPose=@n26, Alpha=1) {
    angularDampingOverride=1
    angularSpringConstant=15000
    bAngularSpring=true
    bOverrideAngularDamping=true
    bUseGravityOverride=true
    gravityScale=0.35
    simulationSpace=World
    alphaBoolBlend {blendOption=Linear}
    boundBone {boneName=dyn_eye_03_L}
    componentLinearAccScale=<1 1 1>
    componentLinearVelScale=<1 1 1>
    gravityOverride {z=-980}
    physicsBodyDefinitions [{
      boundBone {boneName=dyn_eye_03_L}
      boxExtents=<20 20 20>
      constraintSetup {angularLimitsMax=<10 25 25>, angularLimitsMin=<-10 -25 -25>, angularTarget=<1 0 0>}
      localJointOffset=<1 0 0>
     }]
   }
   n28 = Anim Dynamics - Bone: dyn_eye_03_R(ComponentPose=@n27, Alpha=1) {
    angularDampingOverride=1
    angularSpringConstant=15000
    bAngularSpring=true
    bOverrideAngularDamping=true
    bUseGravityOverride=true
    gravityScale=0.35
    simulationSpace=World
    alphaBoolBlend {blendOption=Linear}
    boundBone {boneName=dyn_eye_03_R}
    componentLinearAccScale=<1 1 1>
    componentLinearVelScale=<1 1 1>
    gravityOverride {z=-980}
    physicsBodyDefinitions [{
      boundBone {boneName=dyn_eye_03_R}
      boxExtents=<20 20 20>
      constraintSetup {angularLimitsMax=<10 25 25>, angularLimitsMin=<-10 -25 -25>, angularTarget=<1 0 0>}
      localJointOffset=<1 0 0>
     }]
   }
   Save cached pose 'inFrontEndCache'(Pose=Component To Local(ComponentPose=@n28)) {cacheName=inFrontEndCache}
  ]}
 EventGraph {nodes [
   TryGetPawnOwner()
   Event BlueprintUpdateAnimation()
  ]}
]
variables [
 Gravity Override {type=Vector}
 Gravity Override Params {type=GravityOverrideParamsStruct}
]
